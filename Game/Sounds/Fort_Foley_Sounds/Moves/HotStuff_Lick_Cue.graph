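# Sound cue for a "HotStuff Lick" move foley sound, routed to the player foley sound class.
# A single wave player is the root and plays one fixed sample, HotStuff_Lick03, at reduced volume with no randomization.
# The cue overrides attenuation, with a non-spatialized radius near the source, and overrides concurrency so that a few instances can play, stopping the oldest beyond the limit.
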SoundCue /Game/Sounds/Fort_Foley_Sounds/Moves/HotStuff_Lick_Cue
root=WavePlayer_8
nodes [{id=WavePlayer_8, properties {soundWaveAssetPtr=&/HotStuff_Lick03}}]
properties {
 bOverrideAttenuation=true
 bOverrideConcurrency=true
 maxDistance=4000
 soundClassObject=&Fort_Player_Foley
 volumeMultiplier=0.4
 attenuationOverrides {nonSpatializedRadiusStart=200}
 concurrencyOverrides {maxCount=4, resolutionRule=StopOldest}
}
