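# Foley sound cue for a hot sizzle effect in the Moves set, played through the player foley sound class.
# A mixer layers two sizzle waves at once, with the second at a lower level than the first.
# The cue overrides attenuation and caps concurrent instances, stopping the oldest when the limit is reached.
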SoundCue /Game/Sounds/Fort_Foley_Sounds/Moves/HotStuff_Sizzle_Cue
root=Mixer_1
nodes [
 {id=Mixer_1, children [WavePlayer_6,WavePlayer_7], properties {inputVolume [1,0.4]}}
 {id=WavePlayer_6, properties {soundWaveAssetPtr=&/HotStuff_Sizzle01}}
 {id=WavePlayer_7, properties {soundWaveAssetPtr=&/HotStuff_Sizzle05}}
]
properties {
 bOverrideAttenuation=true
 bOverrideConcurrency=true
 maxDistance=4000
 soundClassObject=&Fort_Player_Foley
 volumeMultiplier=0.8
 attenuationOverrides {nonSpatializedRadiusStart=200}
 concurrencyOverrides {maxCount=4, resolutionRule=StopOldest}
}
