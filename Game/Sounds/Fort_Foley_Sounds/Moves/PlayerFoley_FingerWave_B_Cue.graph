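# Sound cue for player foley on the FingerWave move (B variant), routed to the Fort_Player_Foley sound class.
# A Random node picks among five equally weighted wave variations and feeds a Modulator at the root.
# Attenuation and concurrency are overridden on the cue, limiting it to one playing instance that stops the oldest, at reduced volume.
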SoundCue /Game/Sounds/Fort_Foley_Sounds/Moves/PlayerFoley_FingerWave_B_Cue
root=Modulator_2
nodes [
 {id=Modulator_2, children [Random_0]}
 {id=Random_0, children [WavePlayer_0,WavePlayer_1,WavePlayer_2,WavePlayer_3,WavePlayer_4], properties {weights [1,1,1,1,1]}}
 {id=WavePlayer_0, properties {soundWaveAssetPtr=&/PlayerFoley_FingerWave_B_01}}
 {id=WavePlayer_1, properties {soundWaveAssetPtr=&/PlayerFoley_FingerWave_B_02}}
 {id=WavePlayer_2, properties {soundWaveAssetPtr=&/PlayerFoley_FingerWave_B_03}}
 {id=WavePlayer_3, properties {soundWaveAssetPtr=&/PlayerFoley_FingerWave_B_04}}
 {id=WavePlayer_4, properties {soundWaveAssetPtr=&/PlayerFoley_FingerWave_B_05}}
]
properties {
 bOverrideAttenuation=true
 bOverrideConcurrency=true
 maxDistance=4000
 soundClassObject=&Fort_Player_Foley
 volumeMultiplier=0.5
 attenuationOverrides {nonSpatializedRadiusStart=200}
 concurrencyOverrides {maxCount=1, resolutionRule=StopOldest}
}
