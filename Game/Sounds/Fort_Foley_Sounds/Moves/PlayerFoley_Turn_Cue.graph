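# SoundCue for player turning movement foley, routed to the player foley sound class.
# A Random node picks one of three equally weighted turn waves, which then passes through a Modulator at the root.
# The cue overrides attenuation and sets its own concurrency limit, stopping the oldest instance when too many play at once.
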SoundCue /Game/Sounds/Fort_Foley_Sounds/Moves/PlayerFoley_Turn_Cue
root=Modulator_2
nodes [
 {id=Modulator_2, children [Random_0]}
 {id=Random_0, children [WavePlayer_0,WavePlayer_1,WavePlayer_2], properties {weights [1,1,1]}}
 {id=WavePlayer_0, properties {soundWaveAssetPtr=&/PlayerFoley_Turn_01}}
 {id=WavePlayer_1, properties {soundWaveAssetPtr=&/PlayerFoley_Turn_02}}
 {id=WavePlayer_2, properties {soundWaveAssetPtr=&/PlayerFoley_Turn_03}}
]
properties {
 bOverrideAttenuation=true
 bOverrideConcurrency=true
 maxDistance=4000
 soundClassObject=&Fort_Player_Foley
 volumeMultiplier=0.8
 attenuationOverrides {nonSpatializedRadiusStart=200}
 concurrencyOverrides {maxCount=3, resolutionRule=StopOldest}
}
