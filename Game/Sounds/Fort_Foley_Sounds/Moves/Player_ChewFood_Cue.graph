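# Sound cue that plays player food-chewing foley, routed through the Fort_Player_Foley sound class.
# A Random node picks one of three equally weighted chew waves and feeds a root Modulator that varies pitch and volume.
# The cue overrides attenuation with natural-sound distance falloff and a non-spatialized radius.
# Its concurrency override blocks new instances at the limit rather than stopping playing ones.
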SoundCue /Game/Sounds/Fort_Foley_Sounds/Moves/Player_ChewFood_Cue
root=Modulator_2
nodes [
 {id=Modulator_2, children [Random_1]}
 {id=Random_1, children [WavePlayer_4,WavePlayer_5,WavePlayer_6], properties {weights [1,1,1]}}
 {id=WavePlayer_4, properties {soundWaveAssetPtr=&/Player_ChewFood01}}
 {id=WavePlayer_5, properties {soundWaveAssetPtr=&/Player_ChewFood02}}
 {id=WavePlayer_6, properties {soundWaveAssetPtr=&/Player_ChewFood03}}
]
properties {
 bOverrideAttenuation=true
 bOverrideConcurrency=true
 maxDistance=4000
 soundClassObject=&Fort_Player_Foley
 attenuationOverrides {distanceAlgorithm=NaturalSound, nonSpatializedRadiusStart=200}
 concurrencyOverrides {resolutionRule=PreventNew}
}
