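# Sound cue for the player eating food, a foley sound in the consumables sound class.
# A mixer layers one of four random eating recordings, pitch-modulated upward by a varying amount, with a quieter UI health-up sound.
# Overridden attenuation uses natural-sound falloff, and overridden concurrency allows a single instance, blocking new plays while one is active.
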SoundCue /Game/Sounds/Fort_Foley_Sounds/Moves/Player_EatFood_Cue
root=Mixer_6
nodes [
 {id=Mixer_6, children [Modulator_0, WavePlayer_39], properties {inputVolume [1,0.7]}}
 {id=Modulator_0, children [Random_0], properties {pitchMax=1.5, pitchMin=1}}
 {id=Random_0, children [WavePlayer_0,WavePlayer_1,WavePlayer_2,WavePlayer_3], properties {weights [1,1,1,1]}}
 {id=WavePlayer_0, properties {soundWaveAssetPtr=&/Player_EatFood01}}
 {id=WavePlayer_1, properties {soundWaveAssetPtr=&/Player_EatFood02}}
 {id=WavePlayer_2, properties {soundWaveAssetPtr=&/Player_EatFood03}}
 {id=WavePlayer_3, properties {soundWaveAssetPtr=&/Player_EatFood04}}
 {id=WavePlayer_39, properties {soundWaveAssetPtr=/Game/Sounds/Fort_UI_Sounds/FrontEnd/fort_ui_playerhealth_up_01}}
]
properties {
 bOverrideAttenuation=true
 bOverrideConcurrency=true
 maxDistance=4000
 soundClassObject=&Fort_Consumables
 volumeMultiplier=1.5
 attenuationOverrides {distanceAlgorithm=NaturalSound, nonSpatializedRadiusStart=200}
 concurrencyOverrides {maxCount=1, resolutionRule=PreventNew}
}
