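# Sound cue for player foley during the first part of an auto shotgun reload, routed to the Fort_Player_Foley sound class at reduced volume.
# A Random node picks among four equally weighted, reused player run foley waves and feeds a Modulator at the root.
# Attenuation and concurrency are overridden on the cue, with concurrency set to prevent new instances rather than stop playing ones.
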SoundCue /Game/Sounds/Fort_Foley_Sounds/Moves/Player_Foley_AutoShotgun_Reload_Part1_Cue
root=Modulator_0
nodes [
 {id=Modulator_0, children [Random_4]}
 {id=Random_4, children [WavePlayer_19, WavePlayer_1565, WavePlayer_1563, WavePlayer_1562], properties {weights [1,1,1,1]}}
 {id=WavePlayer_1562, properties {soundWaveAssetPtr=&/player_run_foley_12}}
 {id=WavePlayer_1563, properties {soundWaveAssetPtr=&/player_run_foley_10}}
 {id=WavePlayer_1565, properties {soundWaveAssetPtr=&/player_run_foley_03}}
 {id=WavePlayer_19, properties {soundWaveAssetPtr=&/player_run_foley_09}}
]
properties {
 bOverrideAttenuation=true
 bOverrideConcurrency=true
 maxDistance=4000
 soundClassObject=&Fort_Player_Foley
 volumeMultiplier=0.2
 attenuationOverrides {nonSpatializedRadiusStart=200}
 concurrencyOverrides {resolutionRule=PreventNew}
}
