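# Sound cue for the third part of the player's auto shotgun reload foley, played quietly in the player foley sound class.
# A mixer layers a modulated random pick from seven player run foley waves under a louder, briefly delayed random weapon handling sound.
# The delayed layer draws from shotgun handle and open sounds plus a pistol click.
# The cue overrides attenuation and concurrency, preventing new instances once its limit is reached.
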SoundCue /Game/Sounds/Fort_Foley_Sounds/Moves/Player_Foley_AutoShotgun_Reload_Part3_Cue
root=Mixer_0
nodes [
 {id=Delay_0, children [Random_0], properties {delayMax=0.1, delayMin=0.1}}
 {id=Mixer_0, children [Modulator_0,Delay_0], properties {inputVolume [0.6,1.2]}}
 {id=Modulator_0, children [Random_1]}
 {id=Random_0, children [WavePlayer_1544, WavePlayer_1545, WavePlayer_1546], properties {weights [1,1,1]}}
 {
  id=Random_1
  children [WavePlayer_1537, WavePlayer_1538, WavePlayer_1539, WavePlayer_1540, WavePlayer_1541, WavePlayer_1542, WavePlayer_1543]
  properties {weights [1,1,1,1,1,1,1]}
 }
 {id=WavePlayer_1537, properties {soundWaveAssetPtr=&/player_run_foley_01}}
 {id=WavePlayer_1538, properties {soundWaveAssetPtr=&/player_run_foley_04}}
 {id=WavePlayer_1539, properties {soundWaveAssetPtr=&/player_run_foley_05}}
 {id=WavePlayer_1540, properties {soundWaveAssetPtr=&/player_run_foley_06}}
 {id=WavePlayer_1541, properties {soundWaveAssetPtr=&/player_run_foley_08}}
 {id=WavePlayer_1542, properties {soundWaveAssetPtr=&/player_run_foley_09}}
 {id=WavePlayer_1543, properties {soundWaveAssetPtr=&/player_run_foley_11}}
 {id=WavePlayer_1544, properties {soundWaveAssetPtr=/Game/Sounds/Fort_Weapons/Shotgun/shotgun_load_handle_04}}
 {id=WavePlayer_1545, properties {soundWaveAssetPtr=/Game/Sounds/Fort_Weapons/Shotgun/shotgun_load_open_01}}
 {id=WavePlayer_1546, properties {soundWaveAssetPtr=/Game/Sounds/Fort_Weapons/pistol/pistol_click_01}}
]
properties {
 bOverrideAttenuation=true
 bOverrideConcurrency=true
 maxDistance=4000
 soundClassObject=&Fort_Player_Foley
 volumeMultiplier=0.2
 attenuationOverrides {nonSpatializedRadiusStart=200}
 concurrencyOverrides {resolutionRule=PreventNew}
}
