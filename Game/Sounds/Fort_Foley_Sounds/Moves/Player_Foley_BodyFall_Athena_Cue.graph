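# Sound cue for the player body-fall foley in Athena, assigned to the player foley sound class.
# A random node picks one of three equally weighted body-fall waves and feeds it through a modulator at the root.
# The cue plays at slightly reduced volume with its own attenuation and a concurrency override that blocks new instances once the limit is reached.
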SoundCue /Game/Sounds/Fort_Foley_Sounds/Moves/Player_Foley_BodyFall_Athena_Cue
root=Modulator_0
nodes [
 {id=Modulator_0, children [Random_0]}
 {id=Random_0, children [WavePlayer_0,WavePlayer_1,WavePlayer_2], properties {weights [1,1,1]}}
 {id=WavePlayer_0, properties {soundWaveAssetPtr=&/Player_BodyFall_01}}
 {id=WavePlayer_1, properties {soundWaveAssetPtr=&/Player_BodyFall_02}}
 {id=WavePlayer_2, properties {soundWaveAssetPtr=&/Player_BodyFall_03}}
]
properties {
 bOverrideAttenuation=true
 bOverrideConcurrency=true
 maxDistance=4000
 soundClassObject=&Fort_Player_Foley
 volumeMultiplier=0.8
 attenuationOverrides {nonSpatializedRadiusStart=200}
 concurrencyOverrides {resolutionRule=PreventNew}
}
