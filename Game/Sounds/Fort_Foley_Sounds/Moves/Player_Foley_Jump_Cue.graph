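# Sound Cue for player jump foley, routed to the Fort_Player_Foley sound class.
# A Mixer layers a randomly chosen run foley wave with a quieter random swing whoosh pitched up to double.
# Each layer passes through a Modulator after its Random node.
# Attenuation and concurrency are overridden on the cue, allowing one instance at a time and stopping the oldest.
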SoundCue /Game/Sounds/Fort_Foley_Sounds/Moves/Player_Foley_Jump_Cue
root=Mixer_0
nodes [
 {id=Mixer_0, children [Modulator_0,Modulator_1], properties {inputVolume [0.5,0.3]}}
 {id=Modulator_0, children [Random_5]}
 {id=Modulator_1, children [Random_1], properties {pitchMax=2, pitchMin=2}}
 {id=Random_1, children [WavePlayer_289, WavePlayer_290, WavePlayer_291, WavePlayer_292, WavePlayer_293], properties {weights [1,1,1,1,1]}}
 {id=Random_5, children [WavePlayer_288, WavePlayer_287], properties {weights [1,1]}}
 {id=WavePlayer_287, properties {soundWaveAssetPtr=&/player_run_foley_04}}
 {id=WavePlayer_288, properties {soundWaveAssetPtr=&/player_run_foley_05}}
 {id=WavePlayer_289, properties {soundWaveAssetPtr=/Game/Sounds/Fort_Foley_Sounds/Whooshes/player_swing_whoosh_stereo_01}}
 {id=WavePlayer_290, properties {soundWaveAssetPtr=/Game/Sounds/Fort_Foley_Sounds/Whooshes/player_swing_whoosh_stereo_02}}
 {id=WavePlayer_291, properties {soundWaveAssetPtr=/Game/Sounds/Fort_Foley_Sounds/Whooshes/player_swing_whoosh_stereo_03}}
 {id=WavePlayer_292, properties {soundWaveAssetPtr=/Game/Sounds/Fort_Foley_Sounds/Whooshes/player_swing_whoosh_stereo_04}}
 {id=WavePlayer_293, properties {soundWaveAssetPtr=/Game/Sounds/Fort_Foley_Sounds/Whooshes/player_swing_whoosh_stereo_05}}
]
properties {
 bOverrideAttenuation=true
 bOverrideConcurrency=true
 maxDistance=4000
 soundClassObject=&Fort_Player_Foley
 volumeMultiplier=0.3
 attenuationOverrides {nonSpatializedRadiusStart=200}
 concurrencyOverrides {maxCount=1, resolutionRule=StopOldest}
}
