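# Sound cue for player movement foley, a short hard variant built from run foley recordings, routed to the player foley sound class.
# A Mixer layers three wave players simultaneously, with one at half volume, then a Modulator at default settings varies the blended result.
# The cue plays at reduced volume and overrides attenuation inline despite referencing a super-close player attenuation asset.
# Its concurrency override rejects new instances once the limit is reached.
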SoundCue /Game/Sounds/Fort_Foley_Sounds/Moves/Player_Foley_ShortHard2_Cue
root=Modulator_1
nodes [
 {id=Mixer_1, children [WavePlayer_3,WavePlayer_4,WavePlayer_5], properties {inputVolume [1,0.5,1]}}
 {id=Modulator_1, children [Mixer_1]}
 {id=WavePlayer_3, properties {soundWaveAssetPtr=&/player_run_foley_13}}
 {id=WavePlayer_4, properties {soundWaveAssetPtr=&/player_run_foley_15}}
 {id=WavePlayer_5, properties {soundWaveAssetPtr=&/player_run_foley_14}}
]
properties {
 attenuationSettings=/Game/Sounds/Attenuation/Fort_Player_SuperClose_Att
 bOverrideAttenuation=true
 bOverrideConcurrency=true
 maxDistance=4000
 soundClassObject=&Fort_Player_Foley
 volumeMultiplier=0.3
 attenuationOverrides {nonSpatializedRadiusStart=200}
 concurrencyOverrides {resolutionRule=PreventNew}
}
