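# Sound cue for the player's scope zoom-in foley, playing only when the IsEnemy parameter is explicitly false.
# A Mixer layers a quiet random run-foley rustle under a full-level random scope zoom sample, each through its own Modulator, with a raised pitch ceiling on the zoom.
# It uses the weapon foley sound class, has spatialization and attenuation off, and its concurrency override rejects new instances at the limit.
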SoundCue /Game/Sounds/Fort_Foley_Sounds/Moves/Player_Scope_ZoomIn_Cue
root=Branch_0
nodes [
 {id=Branch_0, children [null,Mixer_1,null], properties {boolParameterName=IsEnemy}}
 {id=Mixer_1, children [Modulator_2,Modulator_3], properties {inputVolume [0.1,1]}}
 {id=Modulator_2, children [Random_0]}
 {id=Modulator_3, children [Random_2], properties {pitchMax=1.3}}
 {
  id=Random_0
  children [WavePlayer_6,WavePlayer_5,WavePlayer_4,WavePlayer_3]
  properties {bShouldExcludeFromBranchCulling=true, weights [1,1,1,1]}
 }
 {id=Random_2, children [WavePlayer_7,WavePlayer_8,WavePlayer_9], properties {bShouldExcludeFromBranchCulling=true, weights [1,1,1]}}
 {id=WavePlayer_3, properties {soundWaveAssetPtr=&/player_run_foley_12}}
 {id=WavePlayer_4, properties {soundWaveAssetPtr=&/player_run_foley_10}}
 {id=WavePlayer_5, properties {soundWaveAssetPtr=&/player_run_foley_03}}
 {id=WavePlayer_6, properties {soundWaveAssetPtr=&/player_run_foley_09}}
 {id=WavePlayer_7, properties {soundWaveAssetPtr=&/scope_zoom_in_04}}
 {id=WavePlayer_8, properties {soundWaveAssetPtr=&/scope_zoom_in_01}}
 {id=WavePlayer_9, properties {soundWaveAssetPtr=&/scope_zoom_in_02}}
]
properties {
 bOverrideConcurrency=true
 maxDistance=2097152
 soundClassObject=&Fort_WeaponFoley
 volumeMultiplier=0.65
 attenuationOverrides {
  bAttenuate=false
  bSpatialize=false
  falloffDistance=1500
  nonSpatializedRadiusEnd=50
  nonSpatializedRadiusMode=Direct2D
  nonSpatializedRadiusStart=300
  stereoSpread=300
  attenuationShapeExtents {x=1200}
 }
 concurrencyOverrides {resolutionRule=PreventNew}
}
