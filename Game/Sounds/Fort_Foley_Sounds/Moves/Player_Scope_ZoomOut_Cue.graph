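# Sound cue for the player's scope zoom-out foley, routed to the weapon foley sound class.
# A Random node picks between two reused player run foley waves at equal weight, then a Modulator randomizes pitch from normal to slightly raised.
# The cue plays at very low volume with its own attenuation override and a concurrency override that prevents new instances once the limit is reached.
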SoundCue /Game/Sounds/Fort_Foley_Sounds/Moves/Player_Scope_ZoomOut_Cue
root=Modulator_0
nodes [
 {id=Modulator_0, children [Random_2], properties {pitchMax=1.2, pitchMin=1}}
 {id=Random_2, children [WavePlayer_7,WavePlayer_6], properties {weights [1,1]}}
 {id=WavePlayer_6, properties {soundWaveAssetPtr=&/player_run_foley_05}}
 {id=WavePlayer_7, properties {soundWaveAssetPtr=&/player_run_foley_01}}
]
properties {
 bOverrideAttenuation=true
 bOverrideConcurrency=true
 maxDistance=2700
 soundClassObject=&Fort_WeaponFoley
 volumeMultiplier=0.05
 attenuationOverrides {falloffDistance=1500, nonSpatializedRadiusStart=200, attenuationShapeExtents {x=1200}}
 concurrencyOverrides {resolutionRule=PreventNew}
}
